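# Post-process style animation Blueprint for the Summer Heroes backpack that adds secondary physics motion to its Dyn_Rope bone.
# The input pose passes through a single Anim Dynamics node in component space, then returns to local space as the output.
# The node simulates in actor space with linear and angular springs, heavy damping, reduced gravity and limited linear travel.
# The event graph holds only an unconnected update event.
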AnimBlueprint /BRCosmetics/Accessories/FORT_Backpacks/M_MED_SummerHeroes_Backpack/Meshes/Summer_Heroes_Backpack_Dyn_AnimBp
target_skeleton=&/Summer_Heroes_Backpack_Skeleton
graphs [
 AnimGraph {nodes [
   AnimGraphNode_SubInput_0 = Input Pose()
   n1 = Anim Dynamics - Bone: Dyn_Rope(ComponentPose=Local To Component(LocalPose=@AnimGraphNode_SubInput_0), Alpha=0.082, ExternalForce="0,0,0") {
    alpha=0.9
    angularDampingOverride=0.995
    angularSpringConstant=1500
    bAngularSpring=true
    bGravityOverrideInSimSpace=true
    bLinearSpring=true
    bOverrideAngularDamping=true
    bOverrideLinearDamping=true
    gravityScale=0.25
    linearDampingOverride=0.995
    linearSpringConstant=300
    numSolverIterationsPostUpdate=4
    numSolverIterationsPreUpdate=16
    simulationSpace=Actor
    alphaBoolBlend {blendOption=Linear}
    boundBone {boneName=Dyn_Rope}
    physicsBodyDefinitions [{
      sphereCollisionRadius=0
      boundBone {boneName=Dyn_Rope}
      boxExtents=<15 5 5>
      constraintSetup {
       angularTargetAxis=AxisZ
       twistAxis=AxisY
       angularLimitsMax=<0 0 0>
       angularLimitsMin=<0 0 0>
       angularTarget=<0 0 1>
       linearAxesMax=<10 0 10>
       linearAxesMin=<-10 0 -10>
      }
      localJointOffset=<0 0 0>
     }]
   }
   Output Pose(Result=Component To Local(ComponentPose=@n1)) {layerGroup=None}
  ]}
 EventGraph {nodes [Event BlueprintUpdateAnimation()]}
]
variables []
